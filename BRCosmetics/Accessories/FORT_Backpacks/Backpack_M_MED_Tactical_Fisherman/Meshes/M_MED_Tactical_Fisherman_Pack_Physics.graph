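# Physics asset for the Tactical Fisherman backpack cosmetic, defining collision bodies and joints for its dangling gear.
# Each bone gets a single box body: the large parent pack, two hooks, two nets, a bottle carabiner, gun holster and container.
# All seven accessories constrain directly to dyn_parent in a star layout, with per-item angular limits.
# Hooks and nets lock the first swing axis, while the container and holster have tight limits.
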PhysicsAsset /BRCosmetics/Accessories/FORT_Backpacks/Backpack_M_MED_Tactical_Fisherman/Meshes/M_MED_Tactical_Fisherman_Pack_Physics
bodies {
 dyn_bottle_carabiner [{shapeType=Box, extentX=5.8689, extentY=3.5, extentZ=1.5505, center=<2.4032 -1.7004 0>, rotation=<rot 0 -5.1231 0>}]
 dyn_container [{shapeType=Box, extentX=9.975, extentY=3.7644, extentZ=11.3964, center=<0 -0.2146 0>}]
 dyn_gun_holster [{shapeType=Box, extentX=4.6836, extentY=2.0594, extentZ=4.3318, center=<1.2351 0 -0.0001>}]
 dyn_hook_01 [{shapeType=Box, extentX=4.01, extentY=0.51, extentZ=2.51, center=<1.6795 0 0>}]
 dyn_hook_02 [{shapeType=Box, extentX=4.01, extentY=0.51, extentZ=2.51, center=<1.6795 0 0>}]
 dyn_net_01 [{shapeType=Box, extentX=7.7392, extentY=0.75, extentZ=6.8549, center=<3.1234 1.3635 0.0001>, rotation=<rot 0 29.686 0>}]
 dyn_net_02 [{shapeType=Box, extentX=7.7392, extentY=0.75, extentZ=6.8549, center=<3.1234 1.3635 0.0001>, rotation=<rot 0 29.686 0>}]
 dyn_parent [{shapeType=Box, extentX=30.5, extentY=12.5, extentZ=21.5, center=<0 4.1364 2.042>}]
}
constraints [
 {bone1Name=dyn_hook_01, bone2Name=dyn_parent, swing1LimitDegrees=2, swing1Motion=Locked, swing2LimitDegrees=45, twistLimitDegrees=3}
 {bone1Name=dyn_hook_02, bone2Name=dyn_parent, swing1LimitDegrees=10, swing1Motion=Locked, swing2LimitDegrees=15, twistLimitDegrees=3}
 {bone1Name=dyn_bottle_carabiner, bone2Name=dyn_parent, swing1LimitDegrees=45, swing2LimitDegrees=45, twistLimitDegrees=12}
 {bone1Name=dyn_gun_holster, bone2Name=dyn_parent, swing1LimitDegrees=7, swing2LimitDegrees=7, twistLimitDegrees=5}
 {bone1Name=dyn_container, bone2Name=dyn_parent, swing1LimitDegrees=5, swing2LimitDegrees=5, twistLimitDegrees=2}
 {bone1Name=dyn_net_01, bone2Name=dyn_parent, swing1LimitDegrees=30, swing1Motion=Locked, swing2LimitDegrees=50, twistLimitDegrees=20}
 {bone1Name=dyn_net_02, bone2Name=dyn_parent, swing1LimitDegrees=30, swing1Motion=Locked, swing2LimitDegrees=50, twistLimitDegrees=20}
]
